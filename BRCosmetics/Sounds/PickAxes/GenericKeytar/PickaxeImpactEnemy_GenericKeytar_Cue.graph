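# Sound cue for the Generic Keytar pickaxe's enemy impact, using the shared pickaxe impact attenuation and cosmetic property sheet.
# After a short fixed delay, a weighted random node picks either a plain impact or, less often, a layered one.
# The plain branch plays one of six impact waves; the layered branch mixes one of those with a random half-volume layer wave.
# Each random pick gets slight pitch and volume modulation for variation.
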SoundCue /BRCosmetics/Sounds/PickAxes/GenericKeytar/PickaxeImpactEnemy_GenericKeytar_Cue
root=Delay_0
nodes [
 {id=Delay_0, children [Random_4], properties {delayMax=0.1, delayMin=0.1}}
 {id=Mixer_0, children [Modulator_1,Modulator_2], properties {inputVolume [1,0.5]}}
 {id=Modulator_0, children [Random_1], properties {pitchMax=1.02, pitchMin=0.98, volumeMax=1.02, volumeMin=0.98}}
 {id=Modulator_1, children [Random_2], properties {pitchMax=1.02, pitchMin=0.98, volumeMax=1.02, volumeMin=0.98}}
 {id=Modulator_2, children [Random_3], properties {pitchMax=1.02, pitchMin=0.98, volumeMax=1.02, volumeMin=0.98}}
 {id=Random_1, children [WavePlayer_1,WavePlayer_2,WavePlayer_3,WavePlayer_4,WavePlayer_5,WavePlayer_6], properties {weights [1,1,1,1,1,1]}}
 {
  id=Random_2
  children [WavePlayer_7, WavePlayer_8, WavePlayer_9, WavePlayer_10, WavePlayer_11, WavePlayer_12]
  properties {weights [1,1,1,1,1,1]}
 }
 {
  id=Random_3
  children [WavePlayer_13, WavePlayer_14, WavePlayer_15, WavePlayer_16, WavePlayer_0, WavePlayer_17]
  properties {weights [1,1,1,1,1,1]}
 }
 {id=Random_4, children [Modulator_0,Mixer_0], properties {bRandomizeWithoutReplacement=false, weights [1,0.65]}}
 {id=WavePlayer_0, properties {soundWaveAssetPtr=&/Pickaxe_Generic_Keytar_Impact_05_Lyr}}
 {id=WavePlayer_1, properties {soundWaveAssetPtr=&/Pickaxe_Generic_Keytar_Impact_01}}
 {id=WavePlayer_10, properties {soundWaveAssetPtr=&/Pickaxe_Generic_Keytar_Impact_04}}
 {id=WavePlayer_11, properties {soundWaveAssetPtr=&/Pickaxe_Generic_Keytar_Impact_05}}
 {id=WavePlayer_12, properties {soundWaveAssetPtr=&/Pickaxe_Generic_Keytar_Impact_06}}
 {id=WavePlayer_13, properties {soundWaveAssetPtr=&/Pickaxe_Generic_Keytar_Impact_01_Lyr}}
 {id=WavePlayer_14, properties {soundWaveAssetPtr=&/Pickaxe_Generic_Keytar_Impact_02_Lyr}}
 {id=WavePlayer_15, properties {soundWaveAssetPtr=&/Pickaxe_Generic_Keytar_Impact_03_Lyr}}
 {id=WavePlayer_16, properties {soundWaveAssetPtr=&/Pickaxe_Generic_Keytar_Impact_04_Lyr}}
 {id=WavePlayer_17, properties {soundWaveAssetPtr=&/Pickaxe_Generic_Keytar_Impact_06_Lyr}}
 {id=WavePlayer_2, properties {soundWaveAssetPtr=&/Pickaxe_Generic_Keytar_Impact_02}}
 {id=WavePlayer_3, properties {soundWaveAssetPtr=&/Pickaxe_Generic_Keytar_Impact_03}}
 {id=WavePlayer_4, properties {soundWaveAssetPtr=&Pickaxe_Generic_Keytar_Impact_04}}
 {id=WavePlayer_5, properties {soundWaveAssetPtr=&Pickaxe_Generic_Keytar_Impact_05}}
 {id=WavePlayer_6, properties {soundWaveAssetPtr=&Pickaxe_Generic_Keytar_Impact_06}}
 {id=WavePlayer_7, properties {soundWaveAssetPtr=&Pickaxe_Generic_Keytar_Impact_01}}
 {id=WavePlayer_8, properties {soundWaveAssetPtr=&Pickaxe_Generic_Keytar_Impact_02}}
 {id=WavePlayer_9, properties {soundWaveAssetPtr=&Pickaxe_Generic_Keytar_Impact_03}}
]
properties {
 attenuationSettings=/Game/Sounds/Attenuation/Fort_Pickaxe_Impact_Att
 audioPropertiesSheet=/BRCosmetics/Sounds/Settings/PropertySheets/Pickaxe/APS_Cosmetic_Pickaxe_Impact
 maxDistance=4300
 soundClassObject=&Fort_Impacts
 concurrencySet [null]
}
